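# Physics asset for the arm mesh of the male medium Candy body cosmetic, giving twelve left-shoulder chocolate bones (a to l) limited secondary motion.
# Each piece is a small offset capsule body constrained to its own base capsule, with no constraints linking separate pieces.
# Swing and twist limits are tight and tuned per pair, and piece l's base breaks the naming pattern as dyn_shldr_choco_base_l_l.
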PhysicsAsset /BRCosmetics/Characters/Player/Male/Medium/Bodies/M_MED_Candy/Meshes/M_MED_Candy_Arm_Physics
bodies {
 dyn_shldr_choco_a_base_l [{shapeType=Capsule, length=1, radius=1}]
 dyn_shldr_choco_a_l [{shapeType=Capsule, length=0.5, radius=1.75, center=<0 0 2.7428>}]
 dyn_shldr_choco_b_base_l [{shapeType=Capsule, length=1, radius=1}]
 dyn_shldr_choco_b_l [{shapeType=Capsule, length=0.5, radius=1.75, center=<0 0 2.7428>}]
 dyn_shldr_choco_base_l_l [{shapeType=Capsule, length=1, radius=1}]
 dyn_shldr_choco_c_base_l [{shapeType=Capsule, length=1, radius=1}]
 dyn_shldr_choco_c_l [{shapeType=Capsule, length=0.5, radius=1.75, center=<0 0 2.7428>}]
 dyn_shldr_choco_d_base_l [{shapeType=Capsule, length=1, radius=1}]
 dyn_shldr_choco_d_l [{shapeType=Capsule, length=0.5, radius=1.75, center=<0 0 2.7428>}]
 dyn_shldr_choco_e_base_l [{shapeType=Capsule, length=1, radius=1}]
 dyn_shldr_choco_e_l [{shapeType=Capsule, length=0.5, radius=1.75, center=<0 0 2.7428>}]
 dyn_shldr_choco_f_base_l [{shapeType=Capsule, length=1, radius=1}]
 dyn_shldr_choco_f_l [{shapeType=Capsule, length=0.5, radius=1.75, center=<0 0 2.7428>}]
 dyn_shldr_choco_g_base_l [{shapeType=Capsule, length=1, radius=1}]
 dyn_shldr_choco_g_l [{shapeType=Capsule, length=0.5, radius=1.75, center=<0 0 2.7428>}]
 dyn_shldr_choco_h_base_l [{shapeType=Capsule, length=1, radius=1}]
 dyn_shldr_choco_h_l [{shapeType=Capsule, length=0.5, radius=1.75, center=<0 0 2.7428>}]
 dyn_shldr_choco_i_base_l [{shapeType=Capsule, length=1, radius=1}]
 dyn_shldr_choco_i_l [{shapeType=Capsule, length=0.5, radius=1.75, center=<0 0 2.7428>}]
 dyn_shldr_choco_j_base_l [{shapeType=Capsule, length=1, radius=1}]
 dyn_shldr_choco_j_l [{shapeType=Capsule, length=0.5, radius=1.75, center=<0 0 2.7428>}]
 dyn_shldr_choco_k_base_l [{shapeType=Capsule, length=1, radius=1}]
 dyn_shldr_choco_k_l [{shapeType=Capsule, length=0.5, radius=1.75, center=<0 0 2.7428>}]
 dyn_shldr_choco_l_l [{shapeType=Capsule, length=0.5, radius=1.75, center=<0 0 2.7428>}]
}
constraints [
 {bone1Name=dyn_shldr_choco_f_l, bone2Name=dyn_shldr_choco_f_base_l, swing1LimitDegrees=10, swing2LimitDegrees=10, twistLimitDegrees=10}
 {bone1Name=dyn_shldr_choco_d_l, bone2Name=dyn_shldr_choco_d_base_l, swing1LimitDegrees=30, swing2LimitDegrees=5, twistLimitDegrees=20}
 {bone1Name=dyn_shldr_choco_b_l, bone2Name=dyn_shldr_choco_b_base_l, swing1LimitDegrees=5, swing2LimitDegrees=5, twistLimitDegrees=5}
 {bone1Name=dyn_shldr_choco_a_l, bone2Name=dyn_shldr_choco_a_base_l, swing1LimitDegrees=7, swing2LimitDegrees=15, twistLimitDegrees=10}
 {bone1Name=dyn_shldr_choco_c_l, bone2Name=dyn_shldr_choco_c_base_l, swing1LimitDegrees=5, swing2LimitDegrees=5, twistLimitDegrees=5}
 {bone1Name=dyn_shldr_choco_e_l, bone2Name=dyn_shldr_choco_e_base_l, swing1LimitDegrees=10, swing2LimitDegrees=10, twistLimitDegrees=7}
 {bone1Name=dyn_shldr_choco_g_l, bone2Name=dyn_shldr_choco_g_base_l, swing1LimitDegrees=25, swing2LimitDegrees=15, twistLimitDegrees=5}
 {bone1Name=dyn_shldr_choco_h_l, bone2Name=dyn_shldr_choco_h_base_l, swing1LimitDegrees=5, swing2LimitDegrees=5, twistLimitDegrees=5}
 {bone1Name=dyn_shldr_choco_i_l, bone2Name=dyn_shldr_choco_i_base_l, swing1LimitDegrees=5, swing2LimitDegrees=5, twistLimitDegrees=5}
 {bone1Name=dyn_shldr_choco_j_l, bone2Name=dyn_shldr_choco_j_base_l, swing1LimitDegrees=5, swing2LimitDegrees=5, twistLimitDegrees=5}
 {bone1Name=dyn_shldr_choco_l_l, bone2Name=dyn_shldr_choco_base_l_l, swing1LimitDegrees=5, swing2LimitDegrees=20, twistLimitDegrees=5}
 {bone1Name=dyn_shldr_choco_k_l, bone2Name=dyn_shldr_choco_k_base_l, swing1LimitDegrees=8, swing2LimitDegrees=15, twistLimitDegrees=10}
]
