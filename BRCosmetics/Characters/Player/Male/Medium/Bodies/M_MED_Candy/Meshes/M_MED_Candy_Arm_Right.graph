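# Physics asset for the right arm mesh of the male medium Candy player body cosmetic, covering its "choco" shoulder dynamic bones.
# Twelve independent bone pairs each have a small capsule on the base bone and a wider, offset capsule on the piece bone.
# One constraint joins each piece to its base, with swing and twist limits tuned per pair from nearly rigid to loose, often asymmetric movement.
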PhysicsAsset /BRCosmetics/Characters/Player/Male/Medium/Bodies/M_MED_Candy/Meshes/M_MED_Candy_Arm_Right
bodies {
 dyn_shldr_choco_a_base_r [{shapeType=Capsule, length=1, radius=1}]
 dyn_shldr_choco_a_r [{shapeType=Capsule, length=0.5, radius=1.75, center=<0 0 -2.3368>, rotation=<rot 0 0 180>}]
 dyn_shldr_choco_b_base_r [{shapeType=Capsule, length=1, radius=1}]
 dyn_shldr_choco_b_r [{shapeType=Capsule, length=0.5, radius=1.75, center=<0 0 -2.3368>, rotation=<rot 0 0 180>}]
 dyn_shldr_choco_base_r_r [{shapeType=Capsule, length=1, radius=1}]
 dyn_shldr_choco_c_base_r [{shapeType=Capsule, length=1, radius=1}]
 dyn_shldr_choco_c_r [{shapeType=Capsule, length=0.5, radius=1.75, center=<0 0 -2.3368>, rotation=<rot 0 0 180>}]
 dyn_shldr_choco_d_base_r [{shapeType=Capsule, length=1, radius=1}]
 dyn_shldr_choco_d_r [{shapeType=Capsule, length=0.5, radius=1.75, center=<0 0 -2.3368>, rotation=<rot 0 0 180>}]
 dyn_shldr_choco_e_base_r [{shapeType=Capsule, length=1, radius=1}]
 dyn_shldr_choco_e_r [{shapeType=Capsule, length=0.5, radius=1.75, center=<0 0 -2.3368>, rotation=<rot 0 0 180>}]
 dyn_shldr_choco_f_base_r [{shapeType=Capsule, length=1, radius=1}]
 dyn_shldr_choco_f_r [{shapeType=Capsule, length=0.5, radius=1.75, center=<0 0 -2.3368>, rotation=<rot 0 0 180>}]
 dyn_shldr_choco_g_base_r [{shapeType=Capsule, length=1, radius=1}]
 dyn_shldr_choco_g_r [{shapeType=Capsule, length=0.5, radius=1.75, center=<0 0 -2.3368>, rotation=<rot 0 0 180>}]
 dyn_shldr_choco_h_base_r [{shapeType=Capsule, length=1, radius=1}]
 dyn_shldr_choco_h_r [{shapeType=Capsule, length=0.5, radius=1.75, center=<0 0 -2.3368>, rotation=<rot 0 0 180>}]
 dyn_shldr_choco_i_base_r [{shapeType=Capsule, length=1, radius=1}]
 dyn_shldr_choco_i_r [{shapeType=Capsule, length=0.5, radius=1.75, center=<0 0 -2.3368>, rotation=<rot 0 0 180>}]
 dyn_shldr_choco_j_base_r [{shapeType=Capsule, length=1, radius=1}]
 dyn_shldr_choco_j_r [{shapeType=Capsule, length=0.5, radius=1.75, center=<0 0 -2.3368>, rotation=<rot 0 0 180>}]
 dyn_shldr_choco_k_base_r [{shapeType=Capsule, length=1, radius=1}]
 dyn_shldr_choco_k_r [{shapeType=Capsule, length=0.5, radius=1.75, center=<0 0 -2.3368>, rotation=<rot 0 0 180>}]
 dyn_shldr_choco_r_r [{shapeType=Capsule, length=0.5, radius=1.75, center=<0 0 -2.3368>, rotation=<rot 0 0 180>}]
}
constraints [
 {bone1Name=dyn_shldr_choco_f_r, bone2Name=dyn_shldr_choco_f_base_r, swing1LimitDegrees=45, swing2LimitDegrees=35, twistLimitDegrees=35}
 {bone1Name=dyn_shldr_choco_d_r, bone2Name=dyn_shldr_choco_d_base_r, swing1LimitDegrees=45, swing2LimitDegrees=30, twistLimitDegrees=20}
 {bone1Name=dyn_shldr_choco_b_r, bone2Name=dyn_shldr_choco_b_base_r, swing1LimitDegrees=5, swing2LimitDegrees=5, twistLimitDegrees=5}
 {bone1Name=dyn_shldr_choco_a_r, bone2Name=dyn_shldr_choco_a_base_r, swing1LimitDegrees=10, swing2LimitDegrees=10, twistLimitDegrees=10}
 {bone1Name=dyn_shldr_choco_c_r, bone2Name=dyn_shldr_choco_c_base_r, swing1LimitDegrees=5, swing2LimitDegrees=5, twistLimitDegrees=5}
 {bone1Name=dyn_shldr_choco_e_r, bone2Name=dyn_shldr_choco_e_base_r, swing1LimitDegrees=5, swing2LimitDegrees=5, twistLimitDegrees=5}
 {bone1Name=dyn_shldr_choco_g_r, bone2Name=dyn_shldr_choco_g_base_r, swing1LimitDegrees=5, swing2LimitDegrees=5, twistLimitDegrees=5}
 {bone1Name=dyn_shldr_choco_h_r, bone2Name=dyn_shldr_choco_h_base_r, swing1LimitDegrees=2, swing2LimitDegrees=75, twistLimitDegrees=40}
 {bone1Name=dyn_shldr_choco_i_r, bone2Name=dyn_shldr_choco_i_base_r, swing1LimitDegrees=5, swing2LimitDegrees=5, twistLimitDegrees=5}
 {bone1Name=dyn_shldr_choco_j_r, bone2Name=dyn_shldr_choco_j_base_r, swing1LimitDegrees=60, swing2LimitDegrees=5, twistLimitDegrees=40}
 {bone1Name=dyn_shldr_choco_r_r, bone2Name=dyn_shldr_choco_base_r_r, swing1LimitDegrees=45, swing2LimitDegrees=20, twistLimitDegrees=10}
 {bone1Name=dyn_shldr_choco_k_r, bone2Name=dyn_shldr_choco_k_base_r, swing1LimitDegrees=60, swing2LimitDegrees=25, twistLimitDegrees=40}
]
