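# Physics asset for the leg ribbons on the M_MED_Candy male medium player body cosmetic.
# Each leg has an upper and lower ribbon, each a capsule body constrained to its own small capsule on a base bone.
# Swing and twist limits bound ribbon motion, with much looser twist on the right leg than the left and a narrower second swing on the lower left ribbon.
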PhysicsAsset /BRCosmetics/Characters/Player/Male/Medium/Bodies/M_MED_Candy/Meshes/M_MED_Candy_Leg_Physics
bodies {
 dyn_base_leg_ribbon_base_lo_base_l [{shapeType=Capsule, length=1, radius=1}]
 dyn_base_leg_ribbon_up_base_l [{shapeType=Capsule, length=1, radius=1}]
 dyn_leg_base_ribbon_lo_base_r [{shapeType=Capsule, length=1, radius=1}]
 dyn_leg_base_ribbon_up_base_r [{shapeType=Capsule, length=1, radius=1}]
 dyn_leg_ribbon_lo_l [{shapeType=Capsule, length=6, radius=3, center=<4 0 0>, rotation=<rot 90 0 0>}]
 dyn_leg_ribbon_lo_r [{shapeType=Capsule, length=5.25, radius=3, center=<4 0 0>, rotation=<rot 90 0 0>}]
 dyn_leg_ribbon_up_l [{shapeType=Capsule, length=6, radius=3, center=<4 0 0>, rotation=<rot 90 0 0>}]
 dyn_leg_ribbon_up_r [{shapeType=Capsule, length=5.25, radius=3, center=<4 0 0>, rotation=<rot 90 0 0>}]
}
constraints [
 {
  bone1Name=dyn_leg_ribbon_lo_l
  bone2Name=dyn_base_leg_ribbon_base_lo_base_l
  swing1LimitDegrees=30
  swing2LimitDegrees=15
  twistLimitDegrees=5
 }
 {bone1Name=dyn_leg_ribbon_up_l, bone2Name=dyn_base_leg_ribbon_up_base_l, swing1LimitDegrees=30, swing2LimitDegrees=30, twistLimitDegrees=5}
 {bone1Name=dyn_leg_ribbon_lo_r, bone2Name=dyn_leg_base_ribbon_lo_base_r, swing1LimitDegrees=30, swing2LimitDegrees=30, twistLimitDegrees=45}
 {bone1Name=dyn_leg_ribbon_up_r, bone2Name=dyn_leg_base_ribbon_up_base_r, swing1LimitDegrees=30, swing2LimitDegrees=30, twistLimitDegrees=45}
]
