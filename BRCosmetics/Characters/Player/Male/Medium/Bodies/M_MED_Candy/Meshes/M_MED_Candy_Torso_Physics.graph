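# Physics asset for the torso mesh of the male medium Candy player body cosmetic, covering twelve chocolate pieces (a to l) on the chest.
# Each piece is a zero-length capsule, effectively a sphere, constrained to its own small base capsule.
# Tight swing and twist limits restrict each piece to a slight wobble, with a few pieces allowed more.
# A ribbon base body is also present but has no constraint.
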PhysicsAsset /BRCosmetics/Characters/Player/Male/Medium/Bodies/M_MED_Candy/Meshes/M_MED_Candy_Torso_Physics
bodies {
 dyn_chest_choco_a [{shapeType=Capsule, length=0, radius=2.75, center=<0 0 -1>}]
 dyn_chest_choco_a_base [{shapeType=Capsule, length=1, radius=1}]
 dyn_chest_choco_b [{shapeType=Capsule, length=0, radius=2.75, center=<0 0 -1>}]
 dyn_chest_choco_b_base [{shapeType=Capsule, length=1, radius=1}]
 dyn_chest_choco_base_l [{shapeType=Capsule, length=1, radius=1}]
 dyn_chest_choco_c [{shapeType=Capsule, length=0, radius=2.75, center=<0 0 -1>}]
 dyn_chest_choco_c_base [{shapeType=Capsule, length=1, radius=1}]
 dyn_chest_choco_d [{shapeType=Capsule, length=0, radius=2.75, center=<0 0 -1>}]
 dyn_chest_choco_d_base [{shapeType=Capsule, length=1, radius=1}]
 dyn_chest_choco_e [{shapeType=Capsule, length=0, radius=2.75, center=<0 0 -1>}]
 dyn_chest_choco_e_base [{shapeType=Capsule, length=1, radius=1}]
 dyn_chest_choco_f [{shapeType=Capsule, length=0, radius=2.75, center=<0 0 -1>}]
 dyn_chest_choco_f_base [{shapeType=Capsule, length=1, radius=1}]
 dyn_chest_choco_g [{shapeType=Capsule, length=0, radius=2.75, center=<0 0 -1>}]
 dyn_chest_choco_g_base [{shapeType=Capsule, length=1, radius=1}]
 dyn_chest_choco_h [{shapeType=Capsule, length=0, radius=2.75, center=<0 0 -1>}]
 dyn_chest_choco_h_base [{shapeType=Capsule, length=1, radius=1}]
 dyn_chest_choco_i [{shapeType=Capsule, length=0, radius=2.75, center=<0 0 -1>}]
 dyn_chest_choco_i_base [{shapeType=Capsule, length=1, radius=1}]
 dyn_chest_choco_j [{shapeType=Capsule, length=0, radius=2.75, center=<0 0 -1>}]
 dyn_chest_choco_j_base [{shapeType=Capsule, length=1, radius=1}]
 dyn_chest_choco_k [{shapeType=Capsule, length=0, radius=2.75, center=<0 0 -1>}]
 dyn_chest_choco_k_base [{shapeType=Capsule, length=1, radius=1}]
 dyn_chest_choco_l [{shapeType=Capsule, length=0, radius=2.75, center=<0 0 -1>}]
 dyn_chest_ribbon_a_lo_base [{shapeType=Capsule, length=1, radius=1}]
}
constraints [
 {bone1Name=dyn_chest_choco_f, bone2Name=dyn_chest_choco_f_base, swing1LimitDegrees=5, swing2LimitDegrees=5, twistLimitDegrees=5}
 {bone1Name=dyn_chest_choco_d, bone2Name=dyn_chest_choco_d_base, swing1LimitDegrees=10, swing2LimitDegrees=10, twistLimitDegrees=10}
 {bone1Name=dyn_chest_choco_b, bone2Name=dyn_chest_choco_b_base, swing1LimitDegrees=5, swing2LimitDegrees=5, twistLimitDegrees=5}
 {bone1Name=dyn_chest_choco_a, bone2Name=dyn_chest_choco_a_base, swing1LimitDegrees=5, swing2LimitDegrees=5, twistLimitDegrees=5}
 {bone1Name=dyn_chest_choco_c, bone2Name=dyn_chest_choco_c_base, swing1LimitDegrees=5, swing2LimitDegrees=5, twistLimitDegrees=5}
 {bone1Name=dyn_chest_choco_e, bone2Name=dyn_chest_choco_e_base, swing1LimitDegrees=5, swing2LimitDegrees=5, twistLimitDegrees=5}
 {bone1Name=dyn_chest_choco_i, bone2Name=dyn_chest_choco_i_base, swing1LimitDegrees=5, swing2LimitDegrees=5, twistLimitDegrees=5}
 {bone1Name=dyn_chest_choco_k, bone2Name=dyn_chest_choco_k_base, swing1LimitDegrees=5, swing2LimitDegrees=5, twistLimitDegrees=5}
 {bone1Name=dyn_chest_choco_l, bone2Name=dyn_chest_choco_base_l, swing1LimitDegrees=7, swing2LimitDegrees=7, twistLimitDegrees=7}
 {bone1Name=dyn_chest_choco_j, bone2Name=dyn_chest_choco_j_base, swing1LimitDegrees=5, swing2LimitDegrees=5, twistLimitDegrees=5}
 {bone1Name=dyn_chest_choco_h, bone2Name=dyn_chest_choco_h_base, swing1LimitDegrees=10, swing2LimitDegrees=10, twistLimitDegrees=10}
 {bone1Name=dyn_chest_choco_g, bone2Name=dyn_chest_choco_g_base, swing1LimitDegrees=5, swing2LimitDegrees=5, twistLimitDegrees=5}
]
