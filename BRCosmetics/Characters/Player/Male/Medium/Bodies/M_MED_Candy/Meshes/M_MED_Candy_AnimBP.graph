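# Post-copy AnimBlueprint for the Candy male body cosmetic, adding secondary motion to its ribbons, hearts and shoulder boxes.
# It copies the parent mesh pose, then applies weighted bone constraints and additive offsets driven by joint angles, movement, skydiving, firing and wave noise.
# AnimDynamics and RigidBody nodes add physics, with a crouch-switched sim bone and a backpack-dependent alpha.
# Health-drop logic for hiding chocolate pieces exists but appears unwired.
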AnimBlueprint /BRCosmetics/Characters/Player/Male/Medium/Bodies/M_MED_Candy/Meshes/M_MED_Candy_AnimBP
target_skeleton=&/M_MED_Candy_Skeleton
graphs [
 AnimGraph {nodes [
   n1 = "Transform (Modify) Bone - Bone: dyn_chest_choco_a"(Alpha=1, Translation="0,0,0", Rotation="0,0,0", Scale=Get choco_a_scale() -> (choco_a_scale:Vector)) {
    rotationMode=BMM_Additive
    rotationSpace=BCS_BoneSpace
    scaleMode=BMM_Additive
    scaleSpace=BCS_BoneSpace
    translationMode=BMM_Additive
    translationSpace=BCS_BoneSpace
    alphaBoolBlend {blendOption=Linear}
    boneToModify {boneName=dyn_chest_choco_a}
   }
   n2 = "Transform (Modify) Bone - Bone: dyn_chest_choco_b"(ComponentPose=@n1, Alpha=1, Translation="0,0,0", Rotation="0,0,0", Scale=Get choco_b_scale() -> (choco_b_scale:Vector)) {
    rotationMode=BMM_Additive
    rotationSpace=BCS_BoneSpace
    scaleMode=BMM_Additive
    scaleSpace=BCS_BoneSpace
    translationMode=BMM_Additive
    translationSpace=BCS_BoneSpace
    alphaBoolBlend {blendOption=Linear}
    boneToModify {boneName=dyn_chest_choco_b}
   }
   n3 = "Transform (Modify) Bone - Bone: dyn_chest_choco_c"(ComponentPose=@n2, Alpha=1, Translation="0,0,0", Rotation="0,0,0", Scale=Get choco_c_scale() -> (choco_c_scale:Vector)) {
    rotationMode=BMM_Additive
    rotationSpace=BCS_BoneSpace
    scaleMode=BMM_Additive
    scaleSpace=BCS_BoneSpace
    translationMode=BMM_Additive
    translationSpace=BCS_BoneSpace
    alphaBoolBlend {blendOption=Linear}
    boneToModify {boneName=dyn_chest_choco_c}
   }
   n4 = "Transform (Modify) Bone - Bone: dyn_chest_choco_d"(ComponentPose=@n3, Alpha=1, Translation="0,0,0", Rotation="0,0,0", Scale=Get choco_d_scale() -> (choco_d_scale:Vector)) {
    rotationMode=BMM_Additive
    rotationSpace=BCS_BoneSpace
    scaleMode=BMM_Additive
    scaleSpace=BCS_BoneSpace
    translationMode=BMM_Additive
    translationSpace=BCS_BoneSpace
    alphaBoolBlend {blendOption=Linear}
    boneToModify {boneName=dyn_chest_choco_d}
   }
   n5 = "Transform (Modify) Bone - Bone: dyn_chest_choco_e"(ComponentPose=@n4, Alpha=1, Translation="0,0,0", Rotation="0,0,0", Scale=Get choco_e_scale() -> (choco_e_scale:Vector)) {
    rotationMode=BMM_Additive
    rotationSpace=BCS_BoneSpace
    scaleMode=BMM_Additive
    scaleSpace=BCS_BoneSpace
    translationMode=BMM_Additive
    translationSpace=BCS_BoneSpace
    alphaBoolBlend {blendOption=Linear}
    boneToModify {boneName=dyn_chest_choco_e}
   }
   n6 = "Transform (Modify) Bone - Bone: dyn_chest_choco_f"(ComponentPose=@n5, Alpha=1, Translation="0,0,0", Rotation="0,0,0", Scale=Get choco_f_scale() -> (choco_f_scale:Vector)) {
    rotationMode=BMM_Additive
    rotationSpace=BCS_BoneSpace
    scaleMode=BMM_Additive
    scaleSpace=BCS_BoneSpace
    translationMode=BMM_Additive
    translationSpace=BCS_BoneSpace
    alphaBoolBlend {blendOption=Linear}
    boneToModify {boneName=dyn_chest_choco_f}
   }
   n7 = "Transform (Modify) Bone - Bone: dyn_chest_choco_g"(ComponentPose=@n6, Alpha=1, Translation="0,0,0", Rotation="0,0,0", Scale=Get choco_g_scale() -> (choco_g_scale:Vector)) {
    rotationMode=BMM_Additive
    rotationSpace=BCS_BoneSpace
    scaleMode=BMM_Additive
    scaleSpace=BCS_BoneSpace
    translationMode=BMM_Additive
    translationSpace=BCS_BoneSpace
    alphaBoolBlend {blendOption=Linear}
    boneToModify {boneName=dyn_chest_choco_g}
   }
   n8 = "Transform (Modify) Bone - Bone: dyn_chest_choco_h"(ComponentPose=@n7, Alpha=1, Translation="0,0,0", Rotation="0,0,0", Scale=Get choco_h_scale() -> (choco_h_scale:Vector)) {
    rotationMode=BMM_Additive
    rotationSpace=BCS_BoneSpace
    scaleMode=BMM_Additive
    scaleSpace=BCS_BoneSpace
    translationMode=BMM_Additive
    translationSpace=BCS_BoneSpace
    alphaBoolBlend {blendOption=Linear}
    boneToModify {boneName=dyn_chest_choco_h}
   }
   n9 = "Transform (Modify) Bone - Bone: dyn_chest_choco_i"(ComponentPose=@n8, Alpha=1, Translation="0,0,0", Rotation="0,0,0", Scale=Get choco_i_scale() -> (choco_i_scale:Vector)) {
    rotationMode=BMM_Additive
    rotationSpace=BCS_BoneSpace
    scaleMode=BMM_Additive
    scaleSpace=BCS_BoneSpace
    translationMode=BMM_Additive
    translationSpace=BCS_BoneSpace
    alphaBoolBlend {blendOption=Linear}
    boneToModify {boneName=dyn_chest_choco_i}
   }
   n10 = "Transform (Modify) Bone - Bone: dyn_chest_choco_j"(ComponentPose=@n9, Alpha=1, Translation="0,0,0", Rotation="0,0,0", Scale=Get choco_j_scale() -> (choco_j_scale:Vector)) {
    rotationMode=BMM_Additive
    rotationSpace=BCS_BoneSpace
    scaleMode=BMM_Additive
    scaleSpace=BCS_BoneSpace
    translationMode=BMM_Additive
    translationSpace=BCS_BoneSpace
    alphaBoolBlend {blendOption=Linear}
    boneToModify {boneName=dyn_chest_choco_j}
   }
   n11 = "Transform (Modify) Bone - Bone: dyn_chest_choco_k"(ComponentPose=@n10, Alpha=1, Translation="0,0,0", Rotation="0,0,0", Scale=Get choco_k_scale() -> (choco_k_scale:Vector)) {
    rotationMode=BMM_Additive
    rotationSpace=BCS_BoneSpace
    scaleMode=BMM_Additive
    scaleSpace=BCS_BoneSpace
    translationMode=BMM_Additive
    translationSpace=BCS_BoneSpace
    alphaBoolBlend {blendOption=Linear}
    boneToModify {boneName=dyn_chest_choco_k}
   }
   "Transform (Modify) Bone - Bone: dyn_chest_choco_l"(ComponentPose=@n11, Alpha=1, Translation="0,0,0", Rotation="0,0,0", Scale=Get choco_l_scale() -> (choco_l_scale:Vector)) {
    rotationMode=BMM_Additive
    rotationSpace=BCS_BoneSpace
    scaleMode=BMM_Additive
    scaleSpace=BCS_BoneSpace
    translationMode=BMM_Additive
    translationSpace=BCS_BoneSpace
    alphaBoolBlend {blendOption=Linear}
    boneToModify {boneName=dyn_chest_choco_l}
   }
   n12 = "Constraint - dyn_chest_choco_a_base (PPPPP)"!(
    ComponentPose=Local To Component(LocalPose=Copy Pose From Mesh() {bUseAttachedParent=true})
    Alpha=1
    ConstraintWeights_0=0.004
    ConstraintWeights_1:float=0.0367
    ConstraintWeights_2:float=0.4137
    ConstraintWeights_3:float=0.1502
    ConstraintWeights_4:float=0.3955
   ) -> (Pose) {
    alphaBoolBlend {blendOption=Linear}
    boneToModify {boneName=dyn_chest_choco_a_base}
    constraintSetup [
     {targetBone {boneName=spine_01}}
     {targetBone {boneName=spine_02}}
     {targetBone {boneName=spine_03}}
     {targetBone {boneName=spine_04}}
     {targetBone {boneName=spine_05}}
    ]
    constraintWeights [1,1,1,1,1]
   }
   n13 = "Constraint - dyn_chest_choco_b_base (PPPPP)"!(
    ComponentPose=@n12
    Alpha=1
    ConstraintWeights_0=0.0242
    ConstraintWeights_1:float=0.0975
    ConstraintWeights_2:float=0.4662
    ConstraintWeights_3:float=0.1805
    ConstraintWeights_4:float=0.2316
   ) -> (Pose) {
    alphaBoolBlend {blendOption=Linear}
    boneToModify {boneName=dyn_chest_choco_b_base}
    constraintSetup [
     {targetBone {boneName=spine_01}}
     {targetBone {boneName=spine_02}}
     {targetBone {boneName=spine_03}}
     {targetBone {boneName=spine_04}}
     {targetBone {boneName=spine_05}}
    ]
    constraintWeights [1,1,1,1,1]
   }
   n14 = "Constraint - dyn_chest_choco_c_base (PPPPP)"!(
    ComponentPose=@n13
    Alpha=1
    ConstraintWeights_0=0.0111
    ConstraintWeights_1:float=0.046
    ConstraintWeights_2:float=0.3982
    ConstraintWeights_3:float=0.1683
    ConstraintWeights_4:float=0.3764
   ) -> (Pose) {
    alphaBoolBlend {blendOption=Linear}
    boneToModify {boneName=dyn_chest_choco_c_base}
    constraintSetup [
     {targetBone {boneName=spine_01}}
     {targetBone {boneName=spine_02}}
     {targetBone {boneName=spine_03}}
     {targetBone {boneName=spine_04}}
     {targetBone {boneName=spine_05}}
    ]
    constraintWeights [1,1,1,1,1]
   }
   n15 = "Constraint - dyn_chest_choco_d_base (PPPPP)"!(
    ComponentPose=@n14
    Alpha=1
    ConstraintWeights_0=0.0584
    ConstraintWeights_1:float=0.1494
    ConstraintWeights_2:float=0.4431
    ConstraintWeights_3:float=0.205
    ConstraintWeights_4:float=0.144
   ) -> (Pose) {
    alphaBoolBlend {blendOption=Linear}
    boneToModify {boneName=dyn_chest_choco_d_base}
    constraintSetup [
     {targetBone {boneName=spine_01}}
     {targetBone {boneName=spine_02}}
     {targetBone {boneName=spine_03}}
     {targetBone {boneName=spine_04}}
     {targetBone {boneName=spine_05}}
    ]
    constraintWeights [1,1,1,1,1]
   }
   n16 = "Constraint - dyn_chest_choco_e_base (PPPPP)"!(
    ComponentPose=@n15
    Alpha=1
    ConstraintWeights_0=0.0354
    ConstraintWeights_1:float=0.0801
    ConstraintWeights_2:float=0.387
    ConstraintWeights_3:float=0.1786
    ConstraintWeights_4:float=0.3188
   ) -> (Pose) {
    alphaBoolBlend {blendOption=Linear}
    boneToModify {boneName=dyn_chest_choco_e_base}
    constraintSetup [
     {targetBone {boneName=spine_01}}
     {targetBone {boneName=spine_02}}
     {targetBone {boneName=spine_03}}
     {targetBone {boneName=spine_04}}
     {targetBone {boneName=spine_05}}
    ]
    constraintWeights [1,1,1,1,1]
   }
   n17 = "Constraint - dyn_chest_choco_f_base (PPPPP)"!(
    ComponentPose=@n16
    Alpha=1
    ConstraintWeights_0=0.1454
    ConstraintWeights_1:float=0.2339
    ConstraintWeights_2:float=0.3797
    ConstraintWeights_3:float=0.1756
    ConstraintWeights_4:float=0.0654
   ) -> (Pose) {
    alphaBoolBlend {blendOption=Linear}
    boneToModify {boneName=dyn_chest_choco_f_base}
    constraintSetup [
     {targetBone {boneName=spine_01}}
     {targetBone {boneName=spine_02}}
     {targetBone {boneName=spine_03}}
     {targetBone {boneName=spine_04}}
     {targetBone {boneName=spine_05}}
    ]
    constraintWeights [1,1,1,1,1]
   }
   n18 = "Constraint - dyn_chest_choco_g_base (PPPPP)"!(
    ComponentPose=@n17
    Alpha=1
    ConstraintWeights_0=0.2091
    ConstraintWeights_1:float=0.2854
    ConstraintWeights_2:float=0.3377
    ConstraintWeights_3:float=0.1358
    ConstraintWeights_4:float=0.0319
   ) -> (Pose) {
    alphaBoolBlend {blendOption=Linear}
    boneToModify {boneName=dyn_chest_choco_g_base}
    constraintSetup [
     {targetBone {boneName=spine_01}}
     {targetBone {boneName=spine_02}}
     {targetBone {boneName=spine_03}}
     {targetBone {boneName=spine_04}}
     {targetBone {boneName=spine_05}}
    ]
    constraintWeights [1,1,1,1,1]
   }
   n19 = "Constraint - dyn_chest_choco_h_base (PPPPP)"!(
    ComponentPose=@n18
    Alpha=1
    ConstraintWeights_0=0.1497
    ConstraintWeights_1:float=0.2466
    ConstraintWeights_2:float=0.3688
    ConstraintWeights_3:float=0.171
    ConstraintWeights_4:float=0.0638
   ) -> (Pose) {
    alphaBoolBlend {blendOption=Linear}
    boneToModify {boneName=dyn_chest_choco_h_base}
    constraintSetup [
     {targetBone {boneName=spine_01}}
     {targetBone {boneName=spine_02}}
     {targetBone {boneName=spine_03}}
     {targetBone {boneName=spine_04}}
     {targetBone {boneName=spine_05}}
    ]
    constraintWeights [1,1,1,1,1]
   }
   n20 = "Constraint - dyn_chest_choco_i_base (PPPPP)"!(
    ComponentPose=@n19
    Alpha=1
    ConstraintWeights_0=0.0112
    ConstraintWeights_1:float=0.0457
    ConstraintWeights_2:float=0.3964
    ConstraintWeights_3:float=0.1673
    ConstraintWeights_4:float=0.3794
   ) -> (Pose) {
    alphaBoolBlend {blendOption=Linear}
    boneToModify {boneName=dyn_chest_choco_i_base}
    constraintSetup [
     {targetBone {boneName=spine_01}}
     {targetBone {boneName=spine_02}}
     {targetBone {boneName=spine_03}}
     {targetBone {boneName=spine_04}}
     {targetBone {boneName=spine_05}}
    ]
    constraintWeights [1,1,1,1,1]
   }
   n21 = "Constraint - dyn_chest_choco_j_base (PPPPP)"!(
    ComponentPose=@n20
    Alpha=1
    ConstraintWeights_0=0.0594
    ConstraintWeights_1:float=0.1557
    ConstraintWeights_2:float=0.4433
    ConstraintWeights_3:float=0.2037
    ConstraintWeights_4:float=0.1378
   ) -> (Pose) {
    alphaBoolBlend {blendOption=Linear}
    boneToModify {boneName=dyn_chest_choco_j_base}
    constraintSetup [
     {targetBone {boneName=spine_01}}
     {targetBone {boneName=spine_02}}
     {targetBone {boneName=spine_03}}
     {targetBone {boneName=spine_04}}
     {targetBone {boneName=spine_05}}
    ]
    constraintWeights [1,1,1,1,1]
   }
   n22 = "Constraint - dyn_chest_choco_k_base (PPPPP)"!(
    ComponentPose=@n21
    Alpha=1
    ConstraintWeights_0=0.0044
    ConstraintWeights_1:float=0.0362
    ConstraintWeights_2:float=0.4107
    ConstraintWeights_3:float=0.1479
    ConstraintWeights_4:float=0.4007
   ) -> (Pose) {
    alphaBoolBlend {blendOption=Linear}
    boneToModify {boneName=dyn_chest_choco_k_base}
    constraintSetup [
     {targetBone {boneName=spine_01}}
     {targetBone {boneName=spine_02}}
     {targetBone {boneName=spine_03}}
     {targetBone {boneName=spine_04}}
     {targetBone {boneName=spine_05}}
    ]
    constraintWeights [1,1,1,1,1]
   }
   n23 = "Constraint - dyn_chest_choco_base_l (PPPPP)"!(
    ComponentPose=@n22
    Alpha=1
    ConstraintWeights_0=0.0259
    ConstraintWeights_1:float=0.1032
    ConstraintWeights_2:float=0.4654
    ConstraintWeights_3:float=0.1789
    ConstraintWeights_4:float=0.2266
   ) -> (Pose) {
    alphaBoolBlend {blendOption=Linear}
    boneToModify {boneName=dyn_chest_choco_base_l}
    constraintSetup [
     {targetBone {boneName=spine_01}}
     {targetBone {boneName=spine_02}}
     {targetBone {boneName=spine_03}}
     {targetBone {boneName=spine_04}}
     {targetBone {boneName=spine_05}}
    ]
    constraintWeights [1,1,1,1,1]
   }
   n24 = "Constraint - dyn_chest_ribbon_a_up_base (PPPP)"!(
    ComponentPose=@n23
    Alpha=1
    ConstraintWeights_0=0.0173
    ConstraintWeights_1:float=0.34
    ConstraintWeights_2:float=0.1477
    ConstraintWeights_3:float=0.495
   ) -> (Pose) {
    alphaBoolBlend {blendOption=Linear}
    boneToModify {boneName=dyn_chest_ribbon_a_up_base}
    constraintSetup [{targetBone {boneName=spine_02}}, {targetBone {boneName=spine_03}}, {targetBone {boneName=spine_04}}, {targetBone {boneName=spine_05}}]
    constraintWeights [1,1,1,1]
   }
   n25 = "Constraint - dyn_chest_ribbon_a_lo_base (PPPPPP)"!(
    ComponentPose=@n24
    Alpha=1
    ConstraintWeights_0=0.0989
    ConstraintWeights_1:float=0.2219
    ConstraintWeights_2:float=0.3494
    ConstraintWeights_3:float=0.2306
    ConstraintWeights_4:float=0.0537
    ConstraintWeights_5:float=0.0455
   ) -> (Pose) {
    alphaBoolBlend {blendOption=Linear}
    boneToModify {boneName=dyn_chest_ribbon_a_lo_base}
    constraintSetup [
     {targetBone {boneName=spine_01}}
     {targetBone {boneName=spine_02}}
     {targetBone {boneName=spine_03}}
     {targetBone {boneName=spine_04}}
     {targetBone {boneName=spine_05}}
     {targetBone {boneName=pelvis}}
    ]
    constraintWeights [1,1,1,1,1,1]
   }
   n26 = "Constraint - dyn_leg_base_ribbon_lo_base_r (PPP)"!(ComponentPose=@n25, Alpha=1, ConstraintWeights_0=0.0148, ConstraintWeights_1:float=0.4183, ConstraintWeights_2:float=0.5668) -> (Pose) {
    alphaBoolBlend {blendOption=Linear}
    boneToModify {boneName=dyn_leg_base_ribbon_lo_base_r}
    constraintSetup [{targetBone {boneName=thigh_r}}, {targetBone {boneName=calf_twist_02_r}}, {targetBone {boneName=calf_r}}]
    constraintWeights [1,1,1]
   }
   n27 = "Constraint - dyn_base_leg_ribbon_base_lo_base_l (PP)"!(ComponentPose=@n26, Alpha=1, ConstraintWeights_0=0.4819, ConstraintWeights_1:float=0.5181) -> (Pose) {
    alphaBoolBlend {blendOption=Linear}
    boneToModify {boneName=dyn_base_leg_ribbon_base_lo_base_l}
    constraintSetup [{targetBone {boneName=calf_l}}, {targetBone {boneName=calf_twist_02_l}}]
    constraintWeights [1,1]
   }
   n28 = "Constraint - dyn_hip_ribbon_main_base (PP)"!(ComponentPose=@n27, Alpha=1, ConstraintWeights_0=0.5, ConstraintWeights_1:float=0.5) -> (Pose) {
    alphaBoolBlend {blendOption=Linear}
    boneToModify {boneName=dyn_hip_ribbon_main_base}
    constraintSetup [{targetBone {boneName=pelvis}}, {targetBone {boneName=spine_01}}]
    constraintWeights [1,1]
   }
   n29 = "Transform (Modify) Bone - Bone: dyn_shldr_box_base_l"(ComponentPose=@n28, Alpha=Get ShldrUp_L() -> (ShldrUp_L:double), Translation="4,-2,0.2", Rotation="0,0,-10", Scale="1,1,1") {
    rotationMode=BMM_Additive
    rotationSpace=BCS_BoneSpace
    translationMode=BMM_Additive
    translationSpace=BCS_BoneSpace
    alphaBoolBlend {blendOption=Linear}
    alphaScaleBiasClamp {bMapRange=true, inRange {max=40}}
    boneToModify {boneName=dyn_shldr_box_base_l}
    rotation {roll=-10}
    translation=<4 -2 0.2>
   }
   n30 = "Transform (Modify) Bone - Bone: dyn_shldr_box_base_r"(ComponentPose=@n29, Alpha=Get ShldrUp_R() -> (ShldrUp_R:double), Translation="-4,2,-0.2", Rotation="0,0,-10", Scale="1,1,1") {
    rotationMode=BMM_Additive
    rotationSpace=BCS_BoneSpace
    translationMode=BMM_Additive
    translationSpace=BCS_BoneSpace
    alphaBoolBlend {blendOption=Linear}
    alphaScaleBiasClamp {bMapRange=true, inRange {max=40}}
    boneToModify {boneName=dyn_shldr_box_base_r}
    rotation {roll=-10}
    translation=<-4 2 -0.2>
   }
   n31 = Get bIsSkyDiving()
   n32 = "Transform (Modify) Bone - Bone: dyn_shldr_box_base_l"!(ComponentPose=@n30, bAlphaBoolEnabled:bool=@n31, Translation="4.5,0,0", Rotation="0,0,0", Scale="1,1,1") -> (Pose) {
    alphaInputType=Bool
    rotationMode=BMM_Additive
    rotationSpace=BCS_BoneSpace
    translationMode=BMM_Additive
    translationSpace=BCS_BoneSpace
    alphaBoolBlend {blendOption=Linear}
    alphaScaleBiasClamp {bMapRange=true, inRange {max=40}}
    boneToModify {boneName=dyn_shldr_box_base_l}
    translation {x=4.5}
   }
   n33 = "Transform (Modify) Bone - Bone: dyn_shldr_box_base_r"!(ComponentPose=@n32, bAlphaBoolEnabled:bool=@n31, Translation="-1,0,0", Rotation="0,0,0", Scale="1,1,1") -> (Pose) {
    alphaInputType=Bool
    rotationMode=BMM_Additive
    rotationSpace=BCS_BoneSpace
    translationMode=BMM_Additive
    translationSpace=BCS_BoneSpace
    alphaBoolBlend {blendOption=Linear}
    alphaScaleBiasClamp {bMapRange=true, inRange {max=40}}
    boneToModify {boneName=dyn_shldr_box_base_r}
    translation {x=-1}
   }
   n34 = "Transform (Modify) Bone - Bone: dyn_shldr_box_base_l"!(
    ComponentPose=@n33
    Alpha:float=1
    Translation=Get ShldrBox_gunTrigger_L() -> (ShldrBox_gunTrigger_L:Vector)
    Rotation="0,0,0"
    Scale="1,1,1"
   ) -> (Pose) {
    rotationMode=BMM_Additive
    rotationSpace=BCS_BoneSpace
    translationMode=BMM_Additive
    translationSpace=BCS_BoneSpace
    alphaBoolBlend {blendOption=Linear}
    alphaScaleBiasClamp {inRange {max=40}}
    boneToModify {boneName=dyn_shldr_box_base_l}
   }
   n35 = "Transform (Modify) Bone - Bone: dyn_shldr_box_base_r"!(
    ComponentPose=@n34
    Alpha:float=1
    Translation=Get ShldrBox_gunTrigger_R() -> (ShldrBox_gunTrigger_R:Vector)
    Rotation="0,0,0"
    Scale="1,1,1"
   ) -> (Pose) {
    rotationMode=BMM_Additive
    rotationSpace=BCS_BoneSpace
    translationMode=BMM_Additive
    translationSpace=BCS_BoneSpace
    alphaBoolBlend {blendOption=Linear}
    alphaScaleBiasClamp {inRange {max=40}}
    boneToModify {boneName=dyn_shldr_box_base_r}
   }
   n36 = Get Elbow_L() -> (Elbow_L:double)
   n37 = "Transform (Modify) Bone - Bone: dyn_uparm_heart_mid_base_l"(ComponentPose=@n35, Alpha=@n36, Translation="0,0,-0.5", Rotation="0,0,5", Scale="1,1,1") {
    rotationMode=BMM_Additive
    rotationSpace=BCS_BoneSpace
    translationMode=BMM_Additive
    translationSpace=BCS_BoneSpace
    alphaBoolBlend {blendOption=Linear}
    alphaScaleBiasClamp {bMapRange=true, inRange {max=60}}
    boneToModify {boneName=dyn_uparm_heart_mid_base_l}
    rotation {roll=5}
    translation {z=-0.5}
   }
   n38 = "Transform (Modify) Bone - Bone: dyn_base_loarm_heart_up_base_l"(ComponentPose=@n37, Alpha=@n36, Translation="0,0,0.5", Rotation="0,0,0", Scale="1,1,1") {
    rotationMode=BMM_Additive
    rotationSpace=BCS_BoneSpace
    translationMode=BMM_Additive
    translationSpace=BCS_BoneSpace
    alphaBoolBlend {blendOption=Linear}
    alphaScaleBiasClamp {bMapRange=true, inRange {max=60}}
    boneToModify {boneName=dyn_base_loarm_heart_up_base_l}
    translation {z=0.5}
   }
   n39 = Get Elbow_R() -> (Elbow_R:double)
   n40 = "Transform (Modify) Bone - Bone: dyn_uparm_heart_mid_base_r"(ComponentPose=@n38, Alpha=@n39, Translation="0,0,0.5", Rotation="0,0,5", Scale="1,1,1") {
    rotationMode=BMM_Additive
    rotationSpace=BCS_BoneSpace
    translationMode=BMM_Additive
    translationSpace=BCS_BoneSpace
    alphaBoolBlend {blendOption=Linear}
    alphaScaleBiasClamp {bMapRange=true, inRange {max=60}}
    boneToModify {boneName=dyn_uparm_heart_mid_base_r}
    rotation {roll=5}
    translation {z=0.5}
   }
   n41 = "Transform (Modify) Bone - Bone: dyn_base_roarm_heart_up_base_r"(ComponentPose=@n40, Alpha=@n39, Translation="0,0,-0.5", Rotation="0,0,0", Scale="1,1,1") {
    rotationMode=BMM_Additive
    rotationSpace=BCS_BoneSpace
    translationMode=BMM_Additive
    translationSpace=BCS_BoneSpace
    alphaBoolBlend {blendOption=Linear}
    alphaScaleBiasClamp {bMapRange=true, inRange {max=60}}
    boneToModify {boneName=dyn_base_roarm_heart_up_base_r}
    translation {z=-0.5}
   }
   n42 = Get arm_noise_1() -> (arm_noise_1:Vector)
   n43 = "Transform (Modify) Bone - Bone: dyn_uparm_heart_mid_base_l"(ComponentPose=@n41, Alpha=1, Translation=@n42, Rotation="0,0,0", Scale="1,1,1") {
    rotationMode=BMM_Additive
    rotationSpace=BCS_BoneSpace
    translationMode=BMM_Additive
    translationSpace=BCS_BoneSpace
    alphaBoolBlend {blendOption=Linear}
    boneToModify {boneName=dyn_uparm_heart_mid_base_l}
   }
   n44 = Get arm_noise_2() -> (arm_noise_2:Vector)
   n45 = "Transform (Modify) Bone - Bone: dyn_base_loarm_heart_up_base_l"(ComponentPose=@n43, Alpha=1, Translation=@n44, Rotation="0,0,0", Scale="1,1,1") {
    rotationMode=BMM_Additive
    rotationSpace=BCS_BoneSpace
    translationMode=BMM_Additive
    translationSpace=BCS_BoneSpace
    alphaBoolBlend {blendOption=Linear}
    boneToModify {boneName=dyn_base_loarm_heart_up_base_l}
   }
   n46 = Get arm_noise_3() -> (arm_noise_3:Vector)
   n47 = "Transform (Modify) Bone - Bone: dyn_base_loarm_heart_mid_base_l"(ComponentPose=@n45, Alpha=1, Translation=@n46, Rotation="0,0,0", Scale="1,1,1") {
    rotationMode=BMM_Additive
    rotationSpace=BCS_BoneSpace
    translationMode=BMM_Additive
    translationSpace=BCS_BoneSpace
    alphaBoolBlend {blendOption=Linear}
    boneToModify {boneName=dyn_base_loarm_heart_mid_base_l}
   }
   n48 = Get arm_noise_1()
   n49 = "Transform (Modify) Bone - Bone: dyn_base_loarm_heart_base_lo_base_l"(ComponentPose=@n47, Alpha=1, Translation=@n48, Rotation="0,0,0", Scale="1,1,1") {
    rotationMode=BMM_Additive
    rotationSpace=BCS_BoneSpace
    translationMode=BMM_Additive
    translationSpace=BCS_BoneSpace
    alphaBoolBlend {blendOption=Linear}
    boneToModify {boneName=dyn_base_loarm_heart_base_lo_base_l}
   }
   n50 = "Transform (Modify) Bone - Bone: dyn_uparm_heart_mid_base_r"(ComponentPose=@n49, Alpha=1, Translation=@n42, Rotation="0,0,0", Scale="1,1,1") {
    rotationMode=BMM_Additive
    rotationSpace=BCS_BoneSpace
    translationMode=BMM_Additive
    translationSpace=BCS_BoneSpace
    alphaBoolBlend {blendOption=Linear}
    boneToModify {boneName=dyn_uparm_heart_mid_base_r}
   }
   n51 = "Transform (Modify) Bone - Bone: dyn_base_roarm_heart_up_base_r"(ComponentPose=@n50, Alpha=1, Translation=@n44, Rotation="0,0,0", Scale="1,1,1") {
    rotationMode=BMM_Additive
    rotationSpace=BCS_BoneSpace
    translationMode=BMM_Additive
    translationSpace=BCS_BoneSpace
    alphaBoolBlend {blendOption=Linear}
    boneToModify {boneName=dyn_base_roarm_heart_up_base_r}
   }
   n52 = "Transform (Modify) Bone - Bone: dyn_base_roarm_heart_mid_base_r"(ComponentPose=@n51, Alpha=1, Translation=@n46, Rotation="0,0,0", Scale="1,1,1") {
    rotationMode=BMM_Additive
    rotationSpace=BCS_BoneSpace
    translationMode=BMM_Additive
    translationSpace=BCS_BoneSpace
    alphaBoolBlend {blendOption=Linear}
    boneToModify {boneName=dyn_base_roarm_heart_mid_base_r}
   }
   n53 = "Transform (Modify) Bone - Bone: dyn_base_roarm_heart_base_ro_base_r"(ComponentPose=@n52, Alpha=1, Translation=@n48, Rotation="0,0,0", Scale="1,1,1") {
    rotationMode=BMM_Additive
    rotationSpace=BCS_BoneSpace
    translationMode=BMM_Additive
    translationSpace=BCS_BoneSpace
    alphaBoolBlend {blendOption=Linear}
    boneToModify {boneName=dyn_base_roarm_heart_base_ro_base_r}
   }
   n54 = Get movingFwd() -> (movingFwd:double)
   n55 = "Transform (Modify) Bone - Bone: dyn_chest_ribbon_a_up_base"(ComponentPose=@n53, Alpha=@n54, Translation="0,0,-1", Rotation="0,0,30", Scale="1,1,1") {
    rotationMode=BMM_Additive
    rotationSpace=BCS_BoneSpace
    translationMode=BMM_Additive
    translationSpace=BCS_BoneSpace
    alphaBoolBlend {blendOption=Linear}
    alphaScaleBiasClamp {bInterpResult=true}
    boneToModify {boneName=dyn_chest_ribbon_a_up_base}
    rotation {roll=30}
    translation {z=-1}
   }
   n56 = "Transform (Modify) Bone - Bone: dyn_chest_ribbon_a_lo_base"(ComponentPose=@n55, Alpha=@n54, Translation="0,0,0", Rotation="0,0,-15", Scale="1,1,1") {
    rotationMode=BMM_Additive
    rotationSpace=BCS_BoneSpace
    translationMode=BMM_Additive
    translationSpace=BCS_BoneSpace
    alphaBoolBlend {blendOption=Linear}
    alphaScaleBiasClamp {bInterpResult=true}
    boneToModify {boneName=dyn_chest_ribbon_a_lo_base}
    rotation {roll=-15}
   }
   n57 = "Transform (Modify) Bone - Bone: dyn_chest_ribbon_a_lo_base"(ComponentPose=@n56, Alpha=Get movingBwd() -> (movingBwd:double), Translation="0,0,0", Rotation="0,0,10", Scale="1,1,1") {
    rotationMode=BMM_Additive
    rotationSpace=BCS_BoneSpace
    translationMode=BMM_Additive
    translationSpace=BCS_BoneSpace
    alphaBoolBlend {blendOption=Linear}
    alphaScaleBiasClamp {bInterpResult=true}
    boneToModify {boneName=dyn_chest_ribbon_a_lo_base}
    rotation {roll=10}
   }
   n58 = "Constraint - sim (PP)"!(
    ComponentPose=@n57
    Alpha=1
    ConstraintWeights_0=Get CrouchRootSpace() -> (CrouchRootSpace:double)
    ConstraintWeights_1:float=Get CrouchSpineSpace() -> (CrouchSpineSpace:double)
   ) -> (Pose) {
    alphaBoolBlend {blendOption=Linear}
    boneToModify {boneName=sim}
    constraintSetup [{targetBone {boneName=root}}, {targetBone {boneName=pelvis}}]
    constraintWeights [1,1]
   }
   n59 = Anim Dynamics - Bone: dyn_hip_ribbon_main(ComponentPose=@n58, Alpha=0.3, ExternalForce="0,0,0") {
    angularDampingOverride=1
    angularSpringConstant=5000
    bAngularSpring=true
    bGravityOverrideInSimSpace=true
    bOverrideAngularDamping=true
    gravityScale=0.35
    simulationSpace=RootRelative
    alphaBoolBlend {blendOption=Linear}
    boundBone {boneName=dyn_hip_ribbon_main}
    componentLinearAccScale=<0.3 0.3 0.3>
    componentLinearVelScale=<0.3 0.3 0.3>
    physicsBodyDefinitions [{
      sphereCollisionRadius=0
      boundBone {boneName=dyn_hip_ribbon_main}
      boxExtents=<20 20 20>
      constraintSetup {angularLimitsMax=<0 0 5>, angularLimitsMin=<0 0 -5>, angularTarget=<0 0 0>}
      localJointOffset=<5 0 0>
     }]
   }
   n60 = Anim Dynamics - Bone: dyn_hip_ribbon_l(ComponentPose=@n59, Alpha=1, ExternalForce="0,0,0") {
    angularDampingOverride=1
    angularSpringConstant=1500
    bAngularSpring=true
    bGravityOverrideInSimSpace=true
    bOverrideAngularDamping=true
    gravityScale=0.35
    simulationSpace=RootRelative
    alphaBoolBlend {blendOption=Linear}
    boundBone {boneName=dyn_hip_ribbon_l}
    componentLinearAccScale=<0.5 0.5 0.5>
    componentLinearVelScale=<0.5 0.5 0.5>
    physicsBodyDefinitions [{
      sphereCollisionRadius=0
      boundBone {boneName=dyn_hip_ribbon_l}
      boxExtents=<20 20 20>
      constraintSetup {angularLimitsMax=<15 0 0>, angularLimitsMin=<-15 0 0>, angularTarget=<0 0 0>}
      localJointOffset=<10 0 0>
     }]
   }
   n61 = Anim Dynamics - Bone: dyn_hip_ribbon_r(ComponentPose=@n60, Alpha=1, ExternalForce="0,0,0") {
    angularDampingOverride=1
    angularSpringConstant=1500
    bAngularSpring=true
    bGravityOverrideInSimSpace=true
    bOverrideAngularDamping=true
    gravityScale=0.35
    simulationSpace=RootRelative
    alphaBoolBlend {blendOption=Linear}
    boundBone {boneName=dyn_hip_ribbon_r}
    componentLinearAccScale=<0.5 0.5 0.5>
    componentLinearVelScale=<0.5 0.5 0.5>
    physicsBodyDefinitions [{
      sphereCollisionRadius=0
      boundBone {boneName=dyn_hip_ribbon_r}
      boxExtents=<20 20 20>
      constraintSetup {angularLimitsMax=<15 0 0>, angularLimitsMin=<-15 0 0>, angularTarget=<0 0 0>}
      localJointOffset=<10 0 0>
     }]
   }
   n62 = RigidBody(ComponentPose=@n61) {
    bClampLinearTranslationLimitToRefPose=true
    bOverrideWorldGravity=true
    overridePhysicsAsset=&/M_MED_Candy_Torso_Physics
    simulationSpace=BaseBoneSpace
    alphaBoolBlend {blendOption=Linear}
    baseBoneRef {boneName=root}
    componentLinearAccScale=<0.45 0.45 0.45>
    componentLinearVelScale=<0.45 0.45 0.45>
    overrideWorldGravity {z=-980}
   }
   n63 = RigidBody(ComponentPose=@n62, Alpha=Get IsBackpackOn() -> (IsBackpackOn:double)) {
    bClampLinearTranslationLimitToRefPose=true
    bOverrideWorldGravity=true
    overridePhysicsAsset=&/M_MED_Candy_Arm_Right
    simulationSpace=BaseBoneSpace
    alphaBoolBlend {blendOption=Linear}
    alphaScaleBiasClamp {bMapRange=true, outRange {max=0, min=1}}
    baseBoneRef {boneName=sim}
    componentLinearAccScale=<0.5 0.5 0.4>
    componentLinearVelScale=<0.5 0.5 0.4>
    overrideWorldGravity {z=-980}
   }
   n64 = RigidBody(ComponentPose=@n63) {
    bClampLinearTranslationLimitToRefPose=true
    bOverrideWorldGravity=true
    overridePhysicsAsset=&/M_MED_Candy_Leg_Physics
    simulationSpace=BaseBoneSpace
    alphaBoolBlend {blendOption=Linear}
    baseBoneRef {boneName=pelvis}
    componentLinearAccScale=<0.4 0.4 0.4>
    componentLinearVelScale=<0.4 0.4 0.4>
    overrideWorldGravity {z=-980}
   }
   n65 = RigidBody(ComponentPose=@n64) {
    bClampLinearTranslationLimitToRefPose=true
    bOverrideWorldGravity=true
    overridePhysicsAsset=&/M_MED_Candy_Arm_Physics
    simulationSpace=BaseBoneSpace
    alphaBoolBlend {blendOption=Linear}
    baseBoneRef {boneName=upperarm_l}
    componentLinearAccScale=<0.4 0.4 0.4>
    componentLinearVelScale=<0.4 0.4 0.4>
    overrideWorldGravity {z=-980}
   }
   n66 = RigidBody(ComponentPose=@n65) {
    bClampLinearTranslationLimitToRefPose=true
    bOverrideWorldGravity=true
    overridePhysicsAsset=&M_MED_Candy_Arm_Right
    simulationSpace=BaseBoneSpace
    alphaBoolBlend {blendOption=Linear}
    baseBoneRef {boneName=upperarm_r}
    componentLinearAccScale=<0.4 0.4 0.4>
    componentLinearVelScale=<0.4 0.4 0.4>
    overrideWorldGravity {z=-980}
   }
   n67 = Get ribbon_noise() -> (ribbon_noise:Rotator)
   n68 = "Transform (Modify) Bone - Bone: dyn_leg_ribbon_up_l"(ComponentPose=@n66, Alpha=1, Translation="0,0,0", Rotation=@n67, Scale="1,1,1") {
    rotationMode=BMM_Additive
    rotationSpace=BCS_BoneSpace
    translationMode=BMM_Additive
    translationSpace=BCS_BoneSpace
    alphaBoolBlend {blendOption=Linear}
    boneToModify {boneName=dyn_leg_ribbon_up_l}
   }
   n69 = "Transform (Modify) Bone - Bone: dyn_leg_ribbon_up_r"(ComponentPose=@n68, Alpha=1, Translation="0,0,0", Rotation=@n67, Scale="1,1,1") {
    rotationMode=BMM_Additive
    rotationSpace=BCS_BoneSpace
    translationMode=BMM_Additive
    translationSpace=BCS_BoneSpace
    alphaBoolBlend {blendOption=Linear}
    boneToModify {boneName=dyn_leg_ribbon_up_r}
   }
   n70 = "Transform (Modify) Bone - Bone: dyn_leg_ribbon_lo_r"(ComponentPose=@n69, Alpha=1, Translation="0,0,0", Rotation=@n67, Scale="1,1,1") {
    rotationMode=BMM_Additive
    rotationSpace=BCS_BoneSpace
    translationMode=BMM_Additive
    translationSpace=BCS_BoneSpace
    alphaBoolBlend {blendOption=Linear}
    boneToModify {boneName=dyn_leg_ribbon_lo_r}
   }
   n71 = "Transform (Modify) Bone - Bone: dyn_chest_ribbon_b_up"(ComponentPose=@n70, Alpha=1, Translation="0,0,0", Rotation=@n67, Scale="1,1,1") {
    rotationMode=BMM_Additive
    rotationSpace=BCS_BoneSpace
    translationMode=BMM_Additive
    translationSpace=BCS_BoneSpace
    alphaBoolBlend {blendOption=Linear}
    boneToModify {boneName=dyn_chest_ribbon_b_up}
   }
   n72 = "Transform (Modify) Bone - Bone: dyn_chest_ribbon_b_lo"(ComponentPose=@n71, Alpha=1, Translation="0,0,0", Rotation=@n67, Scale="1,1,1") {
    rotationMode=BMM_Additive
    rotationSpace=BCS_BoneSpace
    translationMode=BMM_Additive
    translationSpace=BCS_BoneSpace
    alphaBoolBlend {blendOption=Linear}
    boneToModify {boneName=dyn_chest_ribbon_b_lo}
   }
   Output Pose(Result=Component To Local(ComponentPose=@n72)) {layerGroup=None}
  ]}
 EventGraph {nodes [
   n1 = Cast To FortPlayerPawnAthena(Object=TryGetPawnOwner())
   n2 = GetHealth!(self:Fort Pawn Object Reference=@n1) -> (ReturnValue:float)
   > n3 = Branch(Condition="float < float"!(A:double=@n2, B:double=Get healthVar() -> (healthVar:double)) -> (ReturnValue:bool) {is=title K2Node_CallFunction})
   > Set healthVar(healthVar:double=@n2) -> (Output_Get:double)
   > Set hitNum(hitNum:int=int + int(A=Get hitNum() -> (hitNum:int=0), B=1) -> (ReturnValue=0) {is=title K2Node_CommutativeAssociativeBinaryOperator}) -> (Output_Get:int=0)
   Set healthVar(execute=@n3.else, healthVar=@n2)
   Event BlueprintUpdateAnimation()
   > n4 = Sequence!(execute) -> (then_0, then_1, then_2:Exec, then_3:Exec)
   > n5 = Sequence()
   n6 = GetOwnerSkeletalMeshComponent()
   n7 = GetDeltaTransformFromRefPose(self=@n6, BoneName=clavicle_l, BaseName=spine_05)
   > Set ShldrUp_L(ShldrUp_L:double=BreakRotator(InRot=BreakTransform(InTransform=@n7).Rotation).Pitch) -> (Output_Get:double)
   n8 = GetDeltaTransformFromRefPose(self=@n6, BoneName=clavicle_r, BaseName=spine_05)
   Set ShldrUp_R(execute=@n5.then_1, ShldrUp_R:double=BreakRotator(InRot=BreakTransform(InTransform=@n8).Rotation).Pitch) -> (Output_Get:double)
   n9 = GetDeltaTransformFromRefPose(self=@n6, BoneName=lowerarm_l, BaseName=upperarm_l)
   Set Elbow_L(execute=@n5.then_2, Elbow_L:double=BreakRotator(InRot=BreakTransform(InTransform=@n9).Rotation).Yaw) -> (Output_Get:double)
   n10 = GetDeltaTransformFromRefPose(self=@n6, BoneName=lowerarm_r, BaseName=upperarm_r)
   Set Elbow_R(execute=@n5.then_3, Elbow_R:double=BreakRotator(InRot=BreakTransform(InTransform=@n10).Rotation).Yaw) -> (Output_Get:double)
   n11 = Sequence!(execute=@n4.then_1) -> (then_0, then_1, then_2)
   n12 = BreakVector(InVec=MakeWaveVector!(Amplitudes:Rotator="0.7,0.8,0.5", Frequencies:Rotator="100,100,100", Offsets:Rotator="30,60,0", bUseNoise:bool=false, Min:Vector, Max:Vector) -> (ReturnValue:Vector))
   > Set arm_noise_1(arm_noise_1:Vector=MakeVector(Y=0, Z=@n12)) -> (Output_Get:Vector)
   > Set ribbon_noise(ribbon_noise:Rotator=MakeWaveRotator!(Amplitudes:Rotator="5,5,5", Frequencies:Rotator="75,75,75", Offsets:Rotator="0,0,0", bUseNoise:bool=false, Min:Vector, Max:Vector) -> (ReturnValue:Rotator)) -> (Output_Get:Rotator)
   Set arm_noise_2(execute=@n11.then_1, arm_noise_2:Vector=MakeVector(Y=0, Z=@n12.Y)) -> (Output_Get:Vector)
   Set arm_noise_3(execute=@n11.then_2, arm_noise_3:Vector=MakeVector(Y=0, Z=@n12.Z)) -> (Output_Get:Vector)
   n13 = Sequence!(execute=@n4.then_2) -> (then_0, then_1, then_2, then_3:Exec, then_4:Exec)
   n14 = Sequence(execute=@n13.then_3)
   > Reroute Node()
   Reroute Node(InputPin=@n14.then_1)
   Sequence(execute=@n13.then_2)
   n15 = BreakVector(InVec=Get PawnVelocity())
   > n16 = Branch(Condition="float > float"(A="Absolute (Float)"(A=@n15), B=5) {is=title K2Node_CallFunction})
   > n17 = Sequence!(execute) -> (then_0, then_1)
   n18 = "Absolute (Float)"(A=GetOwnerDirection!(self:Custom Character Part Anim Instance Object Reference) -> (ReturnValue:float))
   > Set movingFwd(movingFwd:double=MapRangeClamped(Value=@n18, InRangeA=0, InRangeB=60, OutRangeA=1, OutRangeB=0)) -> (Output_Get:double)
   Set movingBwd(execute=@n17.then_1, movingBwd:double=MapRangeClamped(Value=@n18, InRangeA=150, InRangeB=180)) -> (Output_Get:double)
   n19 = Sequence(execute=@n16.else)
   > Set movingFwd(movingFwd=0)
   Set movingBwd(execute=@n19.then_1, movingBwd=0)
   Set ShldrBox_gunTrigger_R(execute=@n13.then_1, ShldrBox_gunTrigger_R:Vector=vector * integer!(A:Vector=Get ShldrBox_gunTrigger_L(), B:int=-1) -> (ReturnValue:Vector)) -> (Output_Get:Vector)
   n20 = Get playerPawn() -> (playerPawn:Fort Player Pawn Athena Object Reference)
   Is Valid(exec=@n13.then_4, InputObject=@n20)
   > Do Once()
   n21 = GetSkeletalMeshForPartType(self=@n20, PartType=Backpack)
   > Set IsBackpackOn(IsBackpackOn:double="To Float (Boolean)"(InBool="Not Equal (Object)"!(A:Object Reference=@n21, B:Object Reference) -> (ReturnValue:bool))) -> (Output_Get:double)
   Branch(execute=@n13.then_0, Condition=Get bRecentlyFired() -> (bRecentlyFired:bool))
   > Set ShldrBox_gunTrigger_L(ShldrBox_gunTrigger_L:Vector=MakeVector(X=-0.8, Y=0)) -> (Output_Get:Vector)
   > RetriggerableDelay(Duration=0.01)
   > Set ShldrBox_gunTrigger_L(ShldrBox_gunTrigger_L=MakeVector(Y=0))
   Event BlueprintInitializeAnimation()
   n22 = TryGetPawnOwner()
   > Is Valid(InputObject=@n22)
   > n23 = Cast To FortPlayerPawnAthena(Object=@n22)
   > Set playerPawn(playerPawn:Fort Player Pawn Athena Object Reference=@n23) -> (Output_Get:Fort Player Pawn Athena Object Reference)
   Event OnEndCrouching()
   > Set CrouchRootSpace(CrouchRootSpace:double=0) -> (Output_Get:double)
   > Set CrouchSpineSpace(CrouchSpineSpace:double=1) -> (Output_Get:double)
   > RetriggerableDelay(Duration=0.5)
   > Set CrouchRootSpace(CrouchRootSpace=1)
   > Set CrouchSpineSpace(CrouchSpineSpace=0)
   > ResetDynamics(InTeleportType=TeleportPhysics)
   Event OnBeginCrouching()
   > Set CrouchRootSpace(CrouchRootSpace=0)
   > Set CrouchSpineSpace(CrouchSpineSpace=1)
   > ResetDynamics(InTeleportType=TeleportPhysics)
   > RetriggerableDelay(Duration=0.5)
   > Set CrouchRootSpace(CrouchRootSpace=1)
   > Set CrouchSpineSpace(CrouchSpineSpace=0)
   n24 = Switch on Int!(execute:Exec, Selection:int=Get hitNum() -> (hitNum=0)) -> (Default:Exec, 0:Exec, 1:Exec, 2:Exec, 3:Exec, 4:Exec, 5:Exec, 6:Exec, 7:Exec, 8:Exec, 9:Exec, 10:Exec, 11:Exec, 12:Exec)
   Set choco_a_scale(execute=@n24.1, choco_a_scale:Vector="0, 0, 0") -> (Output_Get:Vector)
   Set choco_j_scale(execute=@n24.10, choco_j_scale:Vector="0, 0, 0") -> (Output_Get:Vector)
   Set choco_k_scale(execute=@n24.11, choco_k_scale:Vector="0, 0, 0") -> (Output_Get:Vector)
   Set choco_l_scale(execute=@n24.12, choco_l_scale:Vector="0, 0, 0") -> (Output_Get:Vector)
   Set choco_b_scale(execute=@n24.2, choco_b_scale:Vector="0, 0, 0") -> (Output_Get:Vector)
   Set choco_c_scale(execute=@n24.3, choco_c_scale:Vector="0, 0, 0") -> (Output_Get:Vector)
   Set choco_d_scale(execute=@n24.4, choco_d_scale:Vector="0, 0, 0") -> (Output_Get:Vector)
   Set choco_e_scale(execute=@n24.5, choco_e_scale:Vector="0, 0, 0") -> (Output_Get:Vector)
   Set choco_f_scale(execute=@n24.6, choco_f_scale:Vector="0, 0, 0") -> (Output_Get:Vector)
   Set choco_g_scale(execute=@n24.7, choco_g_scale:Vector="0, 0, 0") -> (Output_Get:Vector)
   Set choco_h_scale(execute=@n24.8, choco_h_scale:Vector="0, 0, 0") -> (Output_Get:Vector)
   Set choco_i_scale(execute=@n24.9, choco_i_scale:Vector="0, 0, 0") -> (Output_Get:Vector)
   TryGetPawnOwner()
  ]}
]
variables [
 arm_noise_1 {type=Vector}
 arm_noise_2 {type=Vector}
 arm_noise_3 {type=Vector}
 choco_a_scale {type=Vector, default=<1 1 1>}
 choco_b_scale {type=Vector, default=<1 1 1>}
 choco_c_scale {type=Vector, default=<1 1 1>}
 choco_d_scale {type=Vector, default=<1 1 1>}
 choco_e_scale {type=Vector, default=<1 1 1>}
 choco_f_scale {type=Vector, default=<1 1 1>}
 choco_g_scale {type=Vector, default=<1 1 1>}
 choco_h_scale {type=Vector, default=<1 1 1>}
 choco_i_scale {type=Vector, default=<1 1 1>}
 choco_j_scale {type=Vector, default=<1 1 1>}
 choco_k_scale {type=Vector, default=<1 1 1>}
 choco_l_scale {type=Vector, default=<1 1 1>}
 CrouchRootSpace {type=number, default=1}
 CrouchSpineSpace {type=number}
 Elbow_L {type=number}
 Elbow_R {type=number}
 healthVar {type=number}
 hitNum {type=integer}
 IsBackpackOn {type=number}
 movingBwd {type=number}
 movingFwd {type=number}
 playerPawn {type=FortPlayerPawnAthena}
 ribbon_noise {type=Rotator}
 ShldrBox_gunTrigger_L {type=Vector}
 ShldrBox_gunTrigger_R {type=Vector}
 ShldrUp_L {type=number}
 ShldrUp_R {type=number}
]
