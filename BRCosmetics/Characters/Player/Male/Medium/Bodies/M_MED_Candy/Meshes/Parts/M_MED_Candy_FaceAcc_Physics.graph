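# Physics asset for the dynamic neck ribbon on the face accessory part of the male medium Candy cosmetic body.
# Six small capsule bodies cover the ribbon's base, main bone, left and right pieces, and two tails.
# The main bone swings and twists freely against the base, acting as the hub for the side pieces and tails.
# Those four attach to it with narrow angular limits, with the tails allowed more swing but less twist.
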PhysicsAsset /BRCosmetics/Characters/Player/Male/Medium/Bodies/M_MED_Candy/Meshes/Parts/M_MED_Candy_FaceAcc_Physics
bodies {
 dyn_neck_ribon_l [{shapeType=Capsule, length=0.75, radius=1.25, center=<1.4 0 0>, rotation=<rot 90 0 0>}]
 dyn_neck_ribon_main [{shapeType=Capsule, length=1, radius=1}]
 dyn_neck_ribon_main_base [{shapeType=Capsule, length=1, radius=1}]
 dyn_neck_ribon_r [{shapeType=Capsule, length=0.75, radius=1.25, center=<1.4 0 0>, rotation=<rot 90 0 0>}]
 dyn_neck_ribon_tail_l [{shapeType=Capsule, length=2.25, radius=1.25, center=<2 0 0>, rotation=<rot 90 0 0>}]
 dyn_neck_ribon_tail_r [{shapeType=Capsule, length=2.25, radius=1.25, center=<2 0 0>, rotation=<rot 90 0 0>}]
}
constraints [
 {
  bone1Name=dyn_neck_ribon_main
  bone2Name=dyn_neck_ribon_main_base
  swing1LimitDegrees=45
  swing1Motion=Free
  swing2LimitDegrees=45
  swing2Motion=Free
  twistLimitDegrees=45
  twistMotion=Free
 }
 {bone1Name=dyn_neck_ribon_tail_r, bone2Name=dyn_neck_ribon_main, swing1LimitDegrees=20, swing2LimitDegrees=20, twistLimitDegrees=5}
 {bone1Name=dyn_neck_ribon_tail_l, bone2Name=dyn_neck_ribon_main, swing1LimitDegrees=20, swing2LimitDegrees=20, twistLimitDegrees=5}
 {bone1Name=dyn_neck_ribon_l, bone2Name=dyn_neck_ribon_main, swing1LimitDegrees=5, swing2LimitDegrees=10, twistLimitDegrees=10}
 {bone1Name=dyn_neck_ribon_r, bone2Name=dyn_neck_ribon_main, swing1LimitDegrees=5, swing2LimitDegrees=10, twistLimitDegrees=10}
]
