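# Animation Blueprint for the Candy character's face accessory, which reacts to eliminations with a head rotation and mouth change.
# AnimGraph copies the parent pose, additively rotates dyn_head by an angle chosen from targeting state, then runs RigidBody physics.
# A watched AthenaKills stat increase raises the rotation alpha and changes the mouth material frame, clearing after a retriggerable delay.
# Each update also switches that frame for emotes or back to default.
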AnimBlueprint /BRCosmetics/Characters/Player/Male/Medium/Bodies/M_MED_Candy/Meshes/Parts/M_MED_Candy_FaceAcc_AnimBP
target_skeleton=&/M_MED_Candy_FaceAcc_Skeleton
graphs [
 AnimGraph {nodes [
   n1 = "Transform (Modify) Bone - Bone: dyn_head"(
    ComponentPose=Local To Component(LocalPose=Copy Pose From Mesh() {bCopyCurves=true, bUseAttachedParent=true})
    Alpha=Get headTurnAlpha() -> (headTurnAlpha:double)
    Translation="0,0,0"
    Rotation=SelectRotator!(A:Rotator="0, 0,-150", B:Rotator="0, 0,180", bPickA:bool=Get bIsTargeting()) -> (ReturnValue:Rotator)
    Scale="1,1,1"
   ) {
    rotationMode=BMM_Additive
    rotationSpace=BCS_BoneSpace
    translationMode=BMM_Additive
    translationSpace=BCS_BoneSpace
    alphaBoolBlend {blendOption=Linear}
    alphaScaleBiasClamp {bInterpResult=true}
    boneToModify {boneName=dyn_head}
   }
   n2 = RigidBody(ComponentPose=@n1) {
    bClampLinearTranslationLimitToRefPose=true
    bOverrideWorldGravity=true
    overridePhysicsAsset=&/M_MED_Candy_FaceAcc_Physics
    simulationSpace=BaseBoneSpace
    alphaBoolBlend {blendOption=Linear}
    baseBoneRef {boneName=spine_03}
    componentLinearAccScale=<0.3 0.3 0.3>
    componentLinearVelScale=<0.3 0.3 0.3>
    overrideWorldGravity {z=-980}
   }
   Output Pose(Result=Component To Local(ComponentPose=@n2)) {layerGroup=None}
  ]}
 EventGraph {nodes [
   Event BlueprintUpdateAnimation()
   > n1 = Sequence()
   > n2 = Get!(execute:Exec) -> (then:Exec, else:Exec, pawn:Fort Player Pawn Object Reference)
   n3 = Cast To FortPlayerPawn(execute=[@n11.then, @n2.else], Object=TryGetFortPawn())
   > Set pawn(pawn:Fort Player Pawn Object Reference=@n3) -> (Output_Get:Fort Player Pawn Object Reference)
   > n4 = WatchCosmeticStat!(
    execute:Exec
    Pawn:Fort Player Pawn Object Reference=Get pawn() -> (pawn:Fort Player Pawn Object Reference)
    StatTag:Fort Stat Manager Tag Structure=(TagName="StatManager.AthenaKills")
    DefaultValue:int=0
    bBroadcastInitialValue:bool=true
    InitialValueBroadcastDelay:float=-1
    ChangeValueBroadcastDelay:float=-1
    bConsumeWaitingNotifications:bool=true
   ) -> (then:Exec, OnStatChanged:Exec, StatName:Name, StatValue:int) {is=title K2Node_AsyncAction}
   n5 = "To Float (Integer)"(InInt=@n4.StatValue)
   n6 = "float > float"(A=@n5, B=Get KillCount() -> (KillCount:double)) {is=title K2Node_CallFunction}
   Branch(execute=@n4.OnStatChanged, Condition=@n6)
   > Set Active_Kill(Active_Kill:bool=@n6) -> (Output_Get:bool)
   > Set KillCount(KillCount:double=@n5) -> (Output_Get:double)
   > Branch(Condition=Get Active_Kill() -> (Active_Kill:bool))
   > n7 = Get Dynamic Material Instance(SkeletalMeshComponent=GetPartSkeletalMeshComponent())
   > SetScalarParameterValue(self=@n7, ParameterName=Smile Mouth animation Frame, Value=8)
   > Set headTurnAlpha(headTurnAlpha:double=1) -> (Output_Get:double)
   > RetriggerableDelay(Duration=1.5)
   > Set Active_Kill(Active_Kill=false)
   > Set headTurnAlpha(headTurnAlpha=0)
   n8 = Get Dynamic Material Instance(execute=@n1.then_1, SkeletalMeshComponent=GetPartSkeletalMeshComponent())
   > n9 = Branch(Condition=Get bIsPlayingEmote())
   > SetScalarParameterValue(self=@n8, ParameterName=Smile Mouth animation Frame, Value=8)
   n10 = Branch(execute=@n9.else, Condition=Get Active_Kill())
   SetScalarParameterValue(execute=@n10.else, self=@n8, ParameterName=Smile Mouth animation Frame, Value=3)
   n11 = Event OnCharacterCustomizationLoaded()
  ]}
]
variables [
 Active_Kill {type=boolean}
 headTurnAlpha {type=number}
 KillCount {type=number}
 pawn {type=FortPlayerPawn}
]
